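; Medium oak tree Blueprint (variant J) for the Asteria environment, with an integrated tree planter base.
; The construction script runs the parent's logic, then forces the SM_Apollo_TreePlanter_A component to unit world scale and zero world rotation.
; This keeps the planter's size and orientation fixed regardless of the actor's scale or rotation.
; BeginPlay, overlap and Tick are disabled placeholder events, so the Blueprint adds no runtime logic.
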(event EventBeginPlay
  (|CallParentFunction) ; This node is disabled and will not be called.\nDrag off pins to build functionality.
)

(event Collision|EventActorBeginOverlap
  (params (OtherActor :type Actor%20Object%20Reference))
  (|CallParentFunction OtherActor) ; This node is disabled and will not be called.\nDrag off pins to build functionality.
)

(event EventTick
  (params (DeltaSeconds :type Float%20%28single-precision%29))
  (|CallParentFunction DeltaSeconds) ; This node is disabled and will not be called.\nDrag off pins to build functionality.
)

(fn ConstructionScript
  (bind _sm_apollo_treeplanter_a (Variables|Default|GetSM_Apollo_TreePlanter_A))
  (|CallParentFunction)
  (Transformation|SetWorldScale3D _sm_apollo_treeplanter_a "1.000000,1.000000,1.000000")
  (Transformation|SetWorldRotation _sm_apollo_treeplanter_a "0, 0, 0" false false))
